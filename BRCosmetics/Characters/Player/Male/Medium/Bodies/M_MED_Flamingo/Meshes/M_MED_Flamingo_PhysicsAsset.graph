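# Physics asset for the M_MED_Flamingo player body cosmetic, covering only two dynamic belt bones and the lower torso.
# The pelvis and spine_01 each have a capsule body, while Dyn_belt_l and Dyn_belt_r each have a sphere.
# The left belt is constrained to the pelvis and the right belt to spine_01, both with swing and twist locked.
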PhysicsAsset /BRCosmetics/Characters/Player/Male/Medium/Bodies/M_MED_Flamingo/Meshes/M_MED_Flamingo_PhysicsAsset
bodies {
 Dyn_belt_l [{shapeType=Sphere, radius=2.1145, center=<-1.0271 1.5527 0.2543>}]
 Dyn_belt_r [{shapeType=Sphere, radius=3.7795, center=<0.8029 -0.2263 -0.5184>}]
 pelvis [{shapeType=Capsule, length=16.9867, radius=6.3094, center=<-1.3941 -0.1006 0.4176>}]
 spine_01 [{shapeType=Capsule, length=13.3949, radius=5.9773, center=<11.711 0.3999 0>}]
}
constraints [
 {
  bone1Name=Dyn_belt_l
  bone2Name=pelvis
  swing1LimitDegrees=45
  swing1Motion=Locked
  swing2LimitDegrees=45
  swing2Motion=Locked
  twistLimitDegrees=45
  twistMotion=Locked
 }
 {
  bone1Name=Dyn_belt_r
  bone2Name=spine_01
  swing1LimitDegrees=45
  swing1Motion=Locked
  swing2LimitDegrees=45
  swing2Motion=Locked
  twistLimitDegrees=45
  twistMotion=Locked
 }
]
