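# Physics asset for the face accessory mesh of the male medium Flamingo character cosmetic, set up for five feather bones on the head.
# Each Dyn_Feather bone (a through e) has an identical small, flat box body, and the head bone has an angled capsule body.
# Every feather is constrained directly to the head rather than chained, with tight twist and secondary swing limits and a wider primary swing for a mostly single-axis sway.
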PhysicsAsset /BRCosmetics/Characters/Player/Male/Medium/Bodies/M_MED_Flamingo/Meshes/Parts/M_MED_Flamingo_FaceAcc_PhysicsAsset
bodies {
 Dyn_Feather_a [{shapeType=Box, extentX=4, extentY=1, extentZ=4, center=<4 0 0>, rotation=<rot 90 0 0>}]
 Dyn_Feather_b [{shapeType=Box, extentX=4, extentY=1, extentZ=4, center=<4 0 0>, rotation=<rot 90 0 0>}]
 Dyn_Feather_c [{shapeType=Box, extentX=4, extentY=1, extentZ=4, center=<4 0 0>, rotation=<rot 90 0 0>}]
 Dyn_Feather_d [{shapeType=Box, extentX=4, extentY=1, extentZ=4, center=<4 0 0>, rotation=<rot 90 0 0>}]
 Dyn_Feather_e [{shapeType=Box, extentX=4, extentY=1, extentZ=4, center=<4 0 0>, rotation=<rot 90 0 0>}]
 head [{shapeType=Capsule, length=9.8812, radius=5.3947, center=<7.3448 3.6416 -0.2908>, rotation=<rot 0 -19.3873 -89.193>}]
}
constraints [
 {bone1Name=Dyn_Feather_a, bone2Name=head, swing1LimitDegrees=22.5, swing2LimitDegrees=5, twistLimitDegrees=5}
 {bone1Name=Dyn_Feather_e, bone2Name=head, swing1LimitDegrees=22.5, swing2LimitDegrees=5, twistLimitDegrees=5}
 {bone1Name=Dyn_Feather_d, bone2Name=head, swing1LimitDegrees=22.5, swing2LimitDegrees=5, twistLimitDegrees=5}
 {bone1Name=Dyn_Feather_b, bone2Name=head, swing1LimitDegrees=22.5, swing2LimitDegrees=5, twistLimitDegrees=5}
 {bone1Name=Dyn_Feather_c, bone2Name=head, swing1LimitDegrees=22.5, swing2LimitDegrees=5, twistLimitDegrees=5}
]
